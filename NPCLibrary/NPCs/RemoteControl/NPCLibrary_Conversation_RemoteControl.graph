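# Conversation database for the RemoteControl NPC, voiced as Toph, defining the greeting players get when they talk to it.
# The entry point leads to a Speech node whose contextual lines cover a first-ever meeting, reactions to three specific player outfits, and several lines gated by match conversation count.
# First-meeting and outfit lines take their priority from named curve rows.
# The speech then links out to the shared NPCLibrary Quests and BasicServices conversation entries.
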ConversationDatabase /NPCLibrary/NPCs/RemoteControl/NPCLibrary_Conversation_RemoteControl
entry_points [{tag=Conversation.Entry.RemoteControl, destinations [#1]}]
nodes [
 ConversationEntryPointNode_7 {
  id=#1
  class=ConversationEntryPointNode
  outputs [#2]
  properties {entryTag {tagName=Conversation.Entry.RemoteControl}}
 }
 Speech_C_0 {
  id=#2
  class=Speech_C
  outputs [#3, #4]
  properties {
   nodeName=""
   generalConfig {
    defaultMessage="Toph, you rule."
    contextualMessages [
     {
      message=My name's Toph. Sounds like tough.
      contextRequirements [{requirement {class=FortControllerRequirement_IsFirstEverConversationWithNPC}}]
      priority {curve {rowName=Tandem.Speech.Priority.FirstEverInteraction}}
     }
     {
      message=I met your uncle once. He really cares about you.
      contextRequirements [{requirement {class=FortControllerRequirement_HasCID, properties {allowedCIds [/BRCosmetics/Athena/Items/Cosmetics/Characters/Character_TikiTorch]}}}]
      priority {curve {rowName=Tandem.Speech.Priority.OutfitReaction}}
     }
     {
      message=Hey Katara. What do you say we bring back The Runaway?
      contextRequirements [{requirement {class=FortControllerRequirement_HasCID, properties {allowedCIds [/BRCosmetics/Athena/Items/Cosmetics/Characters/Character_DirectContact]}}}]
      priority {curve {rowName=Tandem.Speech.Priority.OutfitReaction}}
     }
     {
      message="What're you doing here, Twinkle Toes?"
      contextRequirements [{requirement {class=FortControllerRequirement_HasCID, properties {allowedCIds [/BRCosmetics/Athena/Items/Cosmetics/Characters/Character_DoubleDuty]}}}]
      priority {curve {rowName=Tandem.Speech.Priority.OutfitReaction}}
     }
     {
      message="Toph, you rule."
      contextRequirements [{requirement {
         class=FortControllerRequirement_CheckNPCMatchConversationCount
         properties {nPCUniqueIdTag {tagName=AISpawnerData.Type.Tandem.RemoteControl}}
        }}]
      priority {curve {rowName=None}}
     }
     {
      message="I'm the greatest Earthbender who ever lived, and don't you forget it!"
      contextRequirements [{requirement {
         class=FortControllerRequirement_CheckNPCMatchConversationCount
         properties {nPCUniqueIdTag {tagName=AISpawnerData.Type.Tandem.RemoteControl}}
        }}]
      priority {curve {rowName=None}}
     }
     {
      message="I was born blind, but I see with earthbending."
      contextRequirements [{requirement {
         class=FortControllerRequirement_CheckNPCMatchConversationCount
         properties {nPCUniqueIdTag {tagName=AISpawnerData.Type.Tandem.RemoteControl}}
        }}]
      priority {curve {rowName=None}}
     }
     {
      message=Don't worry about me. I carry my own weight.
      contextRequirements [{requirement {
         class=FortControllerRequirement_CheckNPCMatchConversationCount
         properties {nPCUniqueIdTag {tagName=AISpawnerData.Type.Tandem.RemoteControl}}
        }}]
      priority {curve {rowName=None}}
     }
    ]
   }
  }
 }
 ConversationLinkNode_0 {id=#4, class=ConversationLinkNode, properties {remoteEntryTag {tagName=Conversation.Entry.BasicServices}}}
 ConversationLinkNode_1 {id=#3, class=ConversationLinkNode, properties {remoteEntryTag {tagName=Conversation.Entry.NPCLibrary.Quests}}}
]
